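# Emissive screen material for the Creative Quest Device: an animated, tinted map display with pulsing icons, two pings and a sweep line.
# Noise-driven drift and zoom move the UVs of a channel-packed map texture; mobile shading uses plain UVs.
# The sweep also offsets those UVs and brightens outline and grid, while sphere-mask pings, scanlines and a sine-based edge vignette shape the glow.
# Screen Color tints the result, which passes through MF_CreativeMoveTool.
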Material /Game/Creative/Devices/QuestDevice/Materials/M_QuestDeviceScreen
graph {
 properties {bUseMaterialAttributes=true, pixelDepthOffsetMode=PDOM_Legacy, usedShadingModels=Default Lit}
 nodes [
  n1 = Multiply(A=TextureCoordinate()) {constB=0.5}
  n2 = Sine(None=ComponentMask(None=@n1) {r=true})
  n3 = Sine(None=ComponentMask(None=@n1) {g=true})
  n4 = Power(Base=OneMinus(None=Multiply(A=@n2, B=@n3))) {constExponent=8}
  n5 = OneMinus(None=Multiply(A=@n4, B=@n4))
  n6 = TextureCoordinate() {vTiling=15}
  n7 = ConstantBiasScale(None=Frac(None=Panner(Coordinate=@n6) {speedY=1})) {bias=-0.5, scale=2}
  n8 = Multiply(A=ComponentMask(None=Multiply(A=@n7, B=@n7)) {g=true}, B=ScalarParameter() {defaultValue=0.15, parameterName=Scanline Strength}) {constB=0.1}
  n9 = ScalarParameter() {defaultValue=0.2, parameterName=Base Value}
  n10 = VectorParameter() {parameterName="Ping 1 (Pos/MaxRadius/Hardness)", defaultValue=<rgba 0.5 0.5 0.1 0.35>}
  n11 = Time()
  n12 = ScalarParameter() {defaultValue=0.015, parameterName=Scroll Speed}
  n13 = TextureSample(UVs=AppendVector(A=Multiply(A=@n11, B=@n12), B=Constant())) {samplerType=SAMPLERTYPE_LinearColor, texture=/Game/Creative/Devices/QuestDevice/Textures/T_QuestDevice_Noise_M}
  n14 = Multiply(A=Subtract(A=AppendVector(A=@n13.R, B=@n13.R)) {constB=0.5}) {constB=2}
  n15 = ScalarParameter() {defaultValue=0.3, parameterName=Max UV Movement}
  n16 = MaterialFunctionCall(
   UVs=Add(A=TextureCoordinate(), B=Multiply(A=@n14, B=@n15) {constB=0.5}) {constB=0}
   Texture Scale=ScalarParameter() {defaultValue=1.5, parameterName=Map Zoom}
  ) {materialFunction=&ScaleUVsByCenter}
  n17 = ShadingPathSwitch(Default=@n16.UVs, Deferred=@n16.UVs, Mobile=TextureCoordinate()) {desc=Migrated from FeatureLevelSwitch}
  n18 = VectorParameter() {parameterName="Ping 2 (Pos/MaxRadius/Hardness)", defaultValue=<rgba 0.7476 0.2333 0.05 0.35>}
  n19 = Add(
   A=SphereMask(A=@n17, B=AppendVector(A=@n10.R, B=@n10.R), Radius=@n10.B, Hardness=@n10.B)
   B=SphereMask(A=@n17, B=AppendVector(A=@n18.R, B=@n18.R), Radius=@n18.B, Hardness=@n18.B)
  )
  n20 = Add(A=@n19, B=Time())
  n21 = Add(A=Multiply(A=@n19) {constB=0.85}, B=Saturate(None=Sine(None=@n20)))
  n22 = Multiply(A=Multiply(A=@n19, B=Power(Base=@n21) {constExponent=10}), B=ScalarParameter() {defaultValue=0.05, parameterName=Ping Intensity})
  n23 = ScalarParameter() {defaultValue=0.3, parameterName=Grid Intensity}
  n24 = Multiply(A=ComponentMask(None=TextureCoordinate()) {r=true}) {constB=0.75}
  n25 = Multiply(A=Time()) {constB=-0.2}
  n26 = Power(Base=Frac(None=Add(A=@n24, B=@n25))) {constExponent=20}
  n27 = Multiply(A=@n26) {constB=0.02}
  n28 = Constant()
  n29 = TextureSample(UVs=Add(A=@n17, B=AppendVector(A=@n27, B=@n28))) {samplerType=SAMPLERTYPE_LinearColor, texture=/Game/Creative/Devices/QuestDevice/Textures/T_QuestDevice_M}
  n30 = ShadingPathSwitch(Default=@n16.0-1 mask, Deferred=@n16.0-1 mask, Mobile=Constant() {r=1}) {desc=Migrated from FeatureLevelSwitch}
  n31 = Multiply(A=@n29.R, B=@n30)
  n32 = Add(
   A=Multiply(A=ScalarParameter() {defaultValue=0.2, parameterName=Map Intensity}, B=Multiply(A=@n29.G, B=@n30))
   B=Multiply(A=ScalarParameter() {defaultValue=1, parameterName=Outline Intensity}, B=@n31)
  )
  n33 = Sine(None=Multiply(A=Time()) {constB=0.75})
  n34 = Multiply(A=Multiply(A=ScalarParameter() {defaultValue=3, parameterName=Icon Intensity}, B=Add(A=@n33) {constB=1.2}), B=Multiply(A=@n29.A, B=@n30))
  n35 = Multiply(A=Multiply(A=Add(A=@n31, B=@n29.B), B=@n26), B=ScalarParameter() {defaultValue=2, parameterName=Sweep Intensity})
  n36 = Multiply(A=@n26) {constB=0.35}
  n37 = Multiply(A=OneMinus(None=Multiply(A=@n29.B, B=@n23)), B=Add(A=Add(A=@n32, B=@n34), B=Add(A=@n35, B=@n36)))
  n38 = Multiply(A=Multiply(A=@n5, B=Add(A=@n8)), B=Add(A=Multiply(A=@n5, B=@n9), B=Add(A=@n22, B=@n37)))
  n39 = VectorParameter() {parameterName=Screen Color, defaultValue=<rgba 0 1 0.5239 1>}
  MaterialFunctionCall(MA In=MakeMaterialAttributes!(
    BaseColor
    Metallic
    Specular
    Roughness
    Anisotropy
    EmissiveColor=Multiply(A=@n38, B=@n39.RGB)
    Opacity
    OpacityMask
    Normal
    Tangent
    WorldPositionOffset
    SubsurfaceColor
    ClearCoat
    ClearCoatRoughness
    AmbientOcclusion
    Refraction
    CustomizedUVs_0
    CustomizedUVs_1
    CustomizedUVs_2
    CustomizedUVs_3
    CustomizedUVs_4
    CustomizedUVs_5
    CustomizedUVs_6
    CustomizedUVs_7
    PixelDepthOffset
    ShadingModel
    Displacement
   )) {materialFunction=&MF_CreativeMoveTool}
 ]
}
